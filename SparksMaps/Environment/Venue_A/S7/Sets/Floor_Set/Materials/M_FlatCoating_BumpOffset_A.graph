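# Lit, Nanite-enabled floor material for the Venue_A S7 set, faking a flat reflective coating over a bump-offset pattern.
# Diffuse blends with a reflection cubemap, both optionally tinted, then optionally multiplies by a Fresnel-driven iridescent curve-atlas color.
# Named reroutes share bump-offset UVs and an invertible Fresnel mask, with lerps pulling normals toward flat and PBR masks toward constants.
# Optional emissive takes its color from a parameter or the MF_MojoStageColors stage palette.
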Material /SparksMaps/Environment/Venue_A/S7/Sets/Floor_Set/Materials/M_FlatCoating_BumpOffset_A
graph {
 properties {bAutomaticallySetUsageInEditor=false, bUsedWithNanite=true, usedShadingModels=Default Lit}
 nodes [
  Constant3Vector() {constant=<rgba 1 0.8 0.3 0>}
  n1 = TextureSampleParameter2D(UVs=NamedRerouteUsage()) {group=_Textures, parameterName=PBR Masks, texture=/SparksMaps/Environment/Venue_A/S7/Sets/Floor_Set/Textures/T_Floor_Gradient_Pattern_A_S}
  MaterialFunctionCall(Float3=LinearInterpolate(
    A=MaterialFunctionCall(R=@n1.R, G=@n1.R, B=@n1.R) {materialFunction=&AppendMany}.RGB
    B=Constant3Vector() {constant=<rgba 1 0 1 0>}
    Alpha=NamedRerouteUsage()
   )) {materialFunction=&BreakOutFloat3Components}
  n2 = NamedRerouteDeclaration(None=Fresnel(
    ExponentIn=ScalarParameter() {defaultValue=10, group=Fresnel, parameterName=FrenelExponent}
    BaseReflectFractionIn=ScalarParameter() {defaultValue=0.001, group=Fresnel, parameterName=BaseReflectFraction}
    Normal=VertexNormalWS()
   ) {baseReflectFraction=0, exponent=4}) {name=Fresnel, nodeColor=<rgba 0 1 0.5547 1>}
  n3 = StaticSwitchParameter(True=OneMinus(None=@n2), False=@n2) {group=Fresnel, parameterName=Invert Fresnel}
  NamedRerouteDeclaration(None=Saturate(None=Multiply(A=@n3) {constB=2})) {name=FresnelMultiply, nodeColor=<rgba 1 0.2344 0 1>}
  n4 = CurveAtlasRowParameter(CurveTime=Sine(None=Fresnel())) {
   atlas=/SparksMaps/Environment/Venue_A/S7/Sets/Floor_Set/Textures/CA_Irridescent
   curve=/SparksMaps/Environment/Venue_A/S7/Sets/Floor_Set/Textures/C_Irridescent_A
   group=Irridescece
   parameterName=Irridescent Curve
  }
  NamedRerouteDeclaration(None=LinearInterpolate(B=@n4, Alpha=ScalarParameter() {defaultValue=0.75, group=Irridescece, parameterName=Irridescent Intensity}) {constA=1}) {name=Irridescent, nodeColor=<rgba 1 0.6328 0 1>}
  NamedRerouteDeclaration(None=BumpOffset(Height=ScalarParameter() {defaultValue=-1.5, group=_Textures, parameterName=Bump Offset Height}) {heightRatio=0.04}) {name=BumpOffsetUV, nodeColor=<rgba 0.3125 0 1 1>}
  n5 = TextureSampleParameter2D(UVs=NamedRerouteUsage()) {
   group=_Textures
   parameterName=Normal
   samplerType=SAMPLERTYPE_Normal
   texture=/SparksMaps/Environment/Venue_A/S7/Sets/Floor_Set/Textures/T_Floor_Gradient_Pattern_A_N
  }
  n6 = ScalarParameter() {group=_Textures, parameterName=Flatten Normal, sliderMax=1}
  Reroute(None=LinearInterpolate(
    A=MaterialFunctionCall(Normal=@n5.RGB, Flatness=@n6) {materialFunction=&FlattenNormal}.Result
    B=Constant3Vector() {constant=<rgba 0 0 1 0>}
    Alpha=NamedRerouteUsage()
   ))
  n7 = TextureSampleParameter2D(UVs=NamedRerouteUsage()) {group=_Textures, parameterName=Diffuse, texture=/SparksMaps/Environment/Venue_A/S7/Sets/Floor_Set/Textures/T_Floor_Gradient_Pattern_A_D}
  n8 = VectorParameter() {group=Tint, parameterName=BaseColor_Tint_Intensity, sortPriority=36, defaultValue=<rgba 0.1354 0.1354 0.1354 1>}
  n9 = Multiply(A=Multiply(A=@n8.RGB, B=@n7.RGB), B=@n8.A)
  n10 = TextureSampleParameterCube!(UVs=ReflectionVectorWS(), Apply View MipBias) {
   group=_Textures
   parameterName=Reflection Cube Map
   samplerType=SAMPLERTYPE_LinearColor
   texture=/SparksMaps/Environment/Venue_A/S7/Sets/Floor_Set/Textures/T_StageReflection_A
  }
  n11 = VectorParameter() {group=Tint, parameterName=CubeMap_Tint_Intensity, sortPriority=36, defaultValue=<rgba 1 1 1 1>}
  n12 = Multiply(A=Multiply(A=@n11.RGB, B=@n10.RGB), B=@n11.A)
  n13 = LinearInterpolate(
   A=StaticSwitchParameter(True=LinearInterpolate(A=@n7.RGB, B=@n9, Alpha=@n7.RGB), False=@n7.RGB) {group=Tint, parameterName=Enable Tint, sortPriority=35}
   B=StaticSwitchParameter(True=LinearInterpolate(A=@n10.RGB, B=@n12, Alpha=@n10.RGB), False=@n10.RGB) {group=Tint, parameterName=Enable Cubemap Tint, sortPriority=35}
   Alpha=Multiply(A=NamedRerouteUsage(), B=ScalarParameter() {defaultValue=1, group=_Textures, parameterName=Cube Map Intensity, sliderMax=1})
  )
  StaticSwitchParameter(True=Multiply(A=@n13, B=NamedRerouteUsage()), False=@n13) {defaultValue=true, group=Irridescece, parameterName=Use Irredescence}
  n14 = TextureSampleParameter2D(UVs=NamedRerouteUsage()) {group=Emissive, parameterName=Emissive, texture=&T_Floor_Gradient_Pattern_A_D}
  n15 = MaterialFunctionCall!(CustomUPosition) {materialFunction=/SparksStageShow/Systems/StageManager/FXManager/MaterialFunctions/MF_MojoStageColors}
  n16 = VectorParameter() {group=Emissive, parameterName=EmissiveColor, defaultValue=<rgba 1 1 1 1>}
  n17 = Multiply(A=@n16.RGB, B=@n16.RGB)
  n18 = StaticBoolParameter() {group=Emissive, parameterName=UseStagePalette}
  StaticSwitch(
   True=Multiply(A=@n14.RGB, B=StaticSwitch(True=@n15.Primary, False=@n17, Value=@n18))
   False=Constant()
   Value=StaticBoolParameter() {group=Emissive, parameterName=UseEmissive}
  )
 ]
}
